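# Translucent, unlit, two-sided parent material for the glowing effect on Mizutani SE vehicle wheels.
# An infinity-offset function layers the Mizutani texture for fake depth, and its green channel is tinted by Custom Color.
# A time-driven flipbook on UV1 animates that emissive over a faint constant glow.
# A static switch optionally masks opacity with an alpha texture, and the result passes through shared vehicle damage and de-rez functions.
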Material /VehicleCosmetics/Wheels/MizutaniSE/Materials/M_WheelParentEffect_MizutaniSE_01
graph {
 properties {
  bAutomaticallySetUsageInEditor=false
  bUseMaterialAttributes=true
  bUsedWithMeshDeformer=true
  bUsedWithSkeletalMesh=true
  blendMode=BLEND_Translucent
  shadingModel=MSM_Unlit
  twoSided=true
  usedShadingModels=Unlit
 }
 nodes [
  Constant()
  Constant()
  n1 = VectorParameter() {parameterName=Custom Color, defaultValue=<rgba 0.78 0.4 1 1>}
  n2 = Constant() {r=5}
  n3 = MaterialFunctionCall!(
   Texture=TextureObjectParameter() {parameterName=Param, samplerType=SAMPLERTYPE_LinearColor, texture=/VehicleCosmetics/Wheels/MizutaniSE/Textures/T_Mizutani_SE}
   Offset=ScalarParameter() {defaultValue=0.33, parameterName=BumpOffsetAmount}
   UVs
   RotateUVs
   GradientFalloff
   GradientDarkness=ScalarParameter() {defaultValue=0.001, parameterName=Gradient Darkness}
  ) {materialFunction=/VehicleCosmetics/Wheels/ClockGear_SE/Materials/MF_Infinity_OffsetFunction}
  n4 = Multiply(
   A=Multiply(A=Multiply(A=@n1.RGB, B=@n2), B=ComponentMask(None=@n3.Result) {g=true})
   B=Clamp(None=ScalarParameter() {defaultValue=0.5, parameterName=SelectionFloorFade})
  )
  n5 = TextureCoordinate() {coordinateIndex=1}
  n6 = ScalarParameter() {defaultValue=4, parameterName=Columns}
  n7 = ScalarParameter() {defaultValue=4, parameterName=Rows}
  n8 = AppendVector(A=@n6, B=@n7)
  n9 = Multiply(A=@n6, B=@n7)
  n10 = Frac(None=Multiply(A=Time(), B=ScalarParameter() {defaultValue=0.5, parameterName=Speed}))
  n11 = Floor(None=Multiply(A=AppendVector(A=@n9, B=@n7), B=AppendVector(A=@n10, B=@n10)))
  n12 = TextureSampleParameter2D(UVs=Add(A=Divide(A=@n5, B=@n8), B=Divide(A=@n11, B=@n8))) {
   parameterName=Rim Emissive FB
   samplerType=SAMPLERTYPE_LinearColor
   texture=/VehicleCosmetics/Wheels/MizutaniSE/Textures/Wheel_Mizutani_RGB_E
  }
  n13 = Add(
   A=Multiply(A=@n4, B=Constant() {r=0.05})
   B=Multiply(A=@n4, B=StaticComponentMaskParameter(None=@n12.RGB) {defaultR=true, parameterName=RIM Emissive FB RGB})
  )
  n14 = StaticSwitchParameter(
   True=LinearInterpolate(
    A=Constant()
    B=Constant() {r=1}
    Alpha=TextureSample() {samplerType=SAMPLERTYPE_LinearColor, texture=/VehicleCosmetics/Wheels/MizutaniSE/Textures/T_Mizutani_SE_Alpha}.R
   )
   False=Constant() {r=1}
  ) {group=5_Special Effects, parameterName=LOds_Mask}
  MaterialFunctionCall(In=MaterialFunctionCall(Material=SetMaterialAttributes!(MaterialAttributes, Emissive Color=@n13, Opacity=@n14) {attributeSetTypes [#1, #2]}) {materialFunction=&MF_BaseVehicle_Damage}.Result) {materialFunction=&MF_VehicleDeRez_Transition}
 ]
}
