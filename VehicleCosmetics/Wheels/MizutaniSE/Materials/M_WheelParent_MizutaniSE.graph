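# Parent material for the Mizutani SE vehicle wheel, extending MF_WheelParent_Simple with switchable rim and tire glow.
# Rim glow tints a masked, time-driven flipbook with custom color and brightness; optional tire glow, off by default, tints the tire normal map's blue channel.
# Vertex color selects between rim and tire glow and blends it into the parent's emissive.
# The resulting attributes pass through vehicle damage, then de-rez transition functions.
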Material /VehicleCosmetics/Wheels/MizutaniSE/Materials/M_WheelParent_MizutaniSE
graph {
 properties {
  bAutomaticallySetUsageInEditor=false
  bUseMaterialAttributes=true
  bUsedWithMeshDeformer=true
  bUsedWithSkeletalMesh=true
  pixelDepthOffsetMode=PDOM_Legacy
  previewMesh=/VehicleCosmetics/Wheels/MizutaniSE/Meshes/SK_Wheel_MizutaniSE
  usedShadingModels=Default Lit
 }
 nodes [
  n1 = MaterialFunctionCall() {materialFunction=&MF_WheelParent_Simple}
  n2 = GetMaterialAttributes(None=@n1.MaterialAttributes) {attributeGetTypes [#1]}
  n3 = LinearInterpolate(A=@n2.Emissive Color, B=NamedRerouteUsage(), Alpha=OneMinus(None=@n1.VertexColorR))
  MaterialFunctionCall(In=MaterialFunctionCall(Material=SetMaterialAttributes(MaterialAttributes=@n2.MaterialAttributes, Emissive Color=@n3) {attributeSetTypes [#1]}) {materialFunction=&MF_BaseVehicle_Damage}.Result) {materialFunction=&MF_VehicleDeRez_Transition}
  n4 = VectorParameter() {group=Special Effect, parameterName=Custom Color, defaultValue=<rgba 0.2118 0.4883 1 1>}
  n5 = Multiply(A=@n4.RGB, B=ScalarParameter() {defaultValue=1, group=Special Effect, parameterName=Glow_Brightness})
  Multiply(A=@n5)
  n6 = Frac(None=Multiply(A=Time(), B=ScalarParameter() {defaultValue=0.5, group=Effect, parameterName=Speed, sortPriority=2}))
  n7 = ScalarParameter() {defaultValue=4, group=Effect, parameterName=Rows, sortPriority=2}
  n8 = ScalarParameter() {defaultValue=4, group=Effect, parameterName=Columns, sortPriority=2}
  n9 = TextureObjectParameter() {
   group=Effect
   parameterName=Rim_Mizutani_RGB
   samplerType=SAMPLERTYPE_LinearColor
   sortPriority=1
   texture=/VehicleCosmetics/Wheels/Mizutani1/Textures/Wheel_Mizutani1_RGB_E
  }
  n10 = ScalarParameter() {defaultValue=-1, group=Effect, parameterName=MipBias, sortPriority=2}
  n11 = Multiply(
   A=@n5
   B=StaticComponentMaskParameter(None=MaterialFunctionCall("Animation  Phase (0-1)"=@n6, Number of Rows=@n7, Number of Columns=@n8, Texture=@n9, MipBias/Level=@n10) {materialFunction=&FlipBook}.Result) {defaultR=true, group=Effect, parameterName=Rim_Emissive_Mask, sortPriority=0}
  )
  n12 = Constant()
  n13 = Multiply(
   A=TextureSampleParameter2D() {parameterName=Tire_N, samplerType=SAMPLERTYPE_Normal, texture=/VehicleCosmetics/Wheels/MizutaniSE/Textures/T_MizutaniSE_Tire_N}.B
   B=@n4.RGB
  )
  NamedRerouteDeclaration(None=LinearInterpolate(
    A=StaticSwitchParameter(True=@n11, False=@n12) {defaultValue=true, parameterName=RimLights}
    B=StaticSwitchParameter(True=@n13, False=@n12) {parameterName=TireLights}
    Alpha=VertexColor()
   )) {name=Emissive, nodeColor=<rgba 1 0.75 0 1>}
 ]
}
